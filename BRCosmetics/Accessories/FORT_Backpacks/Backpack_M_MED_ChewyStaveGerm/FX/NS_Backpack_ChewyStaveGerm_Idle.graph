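# Niagara idle effect for the ChewyStaveGerm backpack: one Sparkles emitter of short-lived, flickering sprites.
# A socket reader places particles using offsets picked from a hand-authored point array, alongside cylinder and box locations masked by spawn group.
# Spawn rate differs between gameplay and front end, alpha fades in after an emitter-age delay, and time of day and quality level adjust brightness.
# Three sprite renderers accompany separately derived center, corona and flicker flare values.
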NiagaraSystem /BRCosmetics/Accessories/FORT_Backpacks/Backpack_M_MED_ChewyStaveGerm/FX/NS_Backpack_ChewyStaveGerm_Idle
dependencies {
 usedDataInterfaces [
  &NiagaraDataInterfaceColorCurve
  &NiagaraDataInterfaceArrayFloat3
  &NiagaraDataInterfaceCurve
  &NiagaraDataInterfaceArrayFloat
  &NiagaraDataInterfaceSocketReader
 ]
 usedDynamicInputs [/Game/Effects/Niagara/DynamicInputs/TimeOfDay/DaySequenceTimeOfDayColor]
 usedModules [
  &self:SystemSpawnScript.SystemScriptSource.SystemScriptGraph.NiagaraNodeAssignment_1.SetVariables_#1
  &SystemState
  /BRCosmetics/Characters/Player/Female/Medium/Bodies/F_MED_PiperShelf_Pearl/FX/TimeOfDayGlobal
  &self:SystemSpawnScript.SystemScriptSource.SystemScriptGraph.NiagaraNodeAssignment_2.SetVariables_#2
 ]
}
emitters {Sparkles {
  emitterUpdateScript [
   &EmitterState()
   &SpawnRate(
    Spawn Probability=0.25
    SpawnRate=&FN_FrontEndFloatModifier(Gameplay=25, Front End=20)
    Use Spawn Probability=true
   )
  ]
  particleSpawnScript [
   &PartitionParticles(Number Of Partitions=12, Partition Mode:ENiagara_PartitionMode_v2=::NewEnumerator5)
   &InitializeParticle(
    AdjustAlpha=false
    AdjustHue=true
    AdjustSaturation=false
    AdjustValue=false
    Color=&LerpLinearColors(
     StartColor=<rgba 1 1 1 1>
     EndColor=&ColorFromCurve(
      ColorCurve=DataInterface() {propertyValues {
        AlphaCurve {keys [<at 0 1 RCIM_Cubic>]}
        BlueCurve {keys [
          <at 0.3001 0.3853 RCIM_Cubic>
          {arriveTangent=1.7848, interpMode=RCIM_Cubic, leaveTangent=1.7848, time=0.3457, value=1}
          <at 0.6445 1>
          <at 0.6875 2>
         ]}
        GreenCurve {keys [<at 0.3001 1 RCIM_Cubic>,<at 0.3457 1 RCIM_Cubic>,<at 0.6445 1>,<at 0.6875 0.8229>]}
        RedCurve {keys [
          <at 0.3001 0.3853 RCIM_Cubic>
          {arriveTangent=1.7848, interpMode=RCIM_Cubic, leaveTangent=1.7848, time=0.3457, value=1}
          <at 0.6445 1>
          <at 0.6875 0.9761>
         ]}
       }}
      CurveIndex=&RandomRangeFloat(Minimum=0, Maximum=1)
     )
     LerpFactor=0
    )
    Color Mode=::NewEnumerator3
    Lifetime Max=0.4
    Lifetime Min=0.3
    Lifetime Mode=::NewEnumerator1
    Sprite Size Max=<25 25>
    Sprite Size Min=<15 15>
    Sprite Size Mode=::NewEnumerator2
    UsePositionOffset=false
    Write Material Random=true
   )
   &CameraOffset(Camera Offset Amount=5) {module=CameraOffset001}
  ]
  particleUpdateScript [
   &ParticleState()
   &SocketLocation(
    SamplePosOffset=true
    Sampled Position Offset=&AddVectorToPosition(
     Position=<0 0 0>
     Vector=&SelectVectorFromArray(
      Array Sampling Mode=::NewEnumerator0
      Direct Array Index=?
      Vector Selection Array=DataInterface() {propertyValues {FloatData [
         <9.5936 -2.6865 4.0196>
         <10.9755 -1.5945 4.7839>
         <7.0147 -1.9874 4.7839>
         <-0.7461 -2.6665 4.0137>
         <2.3653 -1.6018 4.7839>
         <-11.9468 -2.8061 -9.3296>
         <-11.29 -1.6171 -9.33>
         <-9.5703 -2.6084 0.1812>
         <-9.3188 -2.6084 -1.0943>
         <-8.0581 -2.6084 -3.3124>
         <-10.4947 -2.6084 -6.4334>
         <-9.6815 -2.6084 -7.8689>
         <0.7939 -2.3945 12.5002>
         <0.6967 -2.3488 9.6088>
         <0.6382 -1.7117 9.5501>
         <1.1305 -1.3148 3.9131>
         <-1.9994 -1.3148 3.8248>
         <-1.5732 -1.3148 -3.5847>
         <1.2199 -1.3148 2.1235>
         <-2.1911 -1.3148 2.0799>
         <-1.5479 -1.3148 0.5359>
         <0.8037 -1.3148 -4.8549>
         <0.8437 -1.3148 -1.4034>
         <0.6833 -1.3148 -5.4315>
         <0.9103 -1.3148 -7.0679>
         <-1.2223 -1.3148 -10.3616>
         <0.7933 -1.3148 -10.419>
         <0.0201 -1.3148 -11.5204>
         <-0.3292 -1.3148 -12.0782>
         <11.8673 -2.6084 3.8253>
         <8.8606 -2.6084 -0.4085>
         <8.7199 -2.6084 1.5361>
         <10.1487 -2.6084 0.1675>
         <8.2456 -2.6084 -4.1002>
         <-1.1032 -2.6084 -5.5443>
         <-0.3113 -2.6084 2.8476>
         <0.4324 -2.6084 -3.3316>
         <0.0348 -2.6084 -1.2049>
         <-10.6287 -1.6245 -1.5572>
         <-9.7217 -1.4162 4.1025>
         <0.8215 -2.9215 -11.2869>
         <6.9929 -1.3148 -7.3151>
         <6.8986 -1.3148 -8.5051>
         <7.1985 -1.3148 -8.7424>
         <8.654 -1.3148 -4.7219>
         <10.2352 -1.3148 0.9901>
         <5.1927 -1.3148 -5.0323>
         <7.3398 -1.3148 -2.6624>
         <8.1887 -1.3148 4.067>
         <9.6842 -1.3148 4.228>
         <8.8721 -2.9131 -4.1834>
         <5.6593 -2.9229 -5.5415>
         <6.7853 -1.9799 -0.836>
         <5.6849 -2.3919 -3.5995>
         <4.898 -1.7741 -6.0612>
         <6.0398 -2.3523 -8.749>
         <1.4234 -1.3354 -9.0488>
         <1.4234 -2.0813 -7.748>
         <1.4234 -2.5875 -2.0485>
         <13.9399 -1.3269 4.581>
         <12.0239 -1.4767 1.4659>
         <7.8729 -2.2744 -8.3229>
         <8.5772 -2.5807 -6.6619>
         <-7.6718 -2.7414 -6.7051>
         <-6.0561 -2.6111 -3.5406>
         <-9.8422 -1.3538 -3.5682>
         <-11.6258 -2.3963 -6.8668>
         <-13.6268 -1.6723 -9.0683>
         <-10.3747 -2.9763 -9.048>
         <-9.743 -2.9763 -4.4531>
         <-2.3837 -1.5331 1.5251>
         <-2.3792 -2.3571 -0.7112>
         <-2.3714 -2.1861 -4.6145>
         <2.1826 -2.0461 -13.9754>
         <-0.9105 -2.6084 10.7306>
         <-5.9712 -1.3148 3.8684>
         <-9.3177 -1.3148 3.1963>
         <-9.5441 -1.3148 1.9008>
         <-10.7444 -1.3148 1.1397>
         <-9.6234 -1.3148 -0.9428>
         <-8.4482 -1.3148 -5.2404>
         <-10.302 -1.3148 -6.4979>
         <-11.4973 -1.3148 -8.1346>
         <-7.8208 -1.6912 -0.1439>
         <-7.7168 -2.078 1.8863>
         <-7.2897 -1.4209 2.8866>
         <-10.6499 -2.7659 -0.4488>
         <0.7622 -2.9763 4.5944>
         <-2.0819 -2.9763 3.7888>
         <0.8555 -2.9763 -1.7192>
         <-1.9795 -2.9763 -6.5199>
         <1.3353 -2.9763 -0.7044>
         <1.7773 -2.9763 -13.0532>
         <12.2893 -2.9763 4.5624>
         <10.799 -2.9763 4.1298>
         <7.2125 -2.9763 -8.8748>
         <0.272 -1.3148 12.8161>
         <-0.2039 -1.3148 11.0191>
         <-1.8408 -2.9763 11.5152>
         <0.2812 -2.9763 9.7004>
        ]}}
      Random Seed=0
     )
    )
    Socket Reader=$System.Socket Reader
    Socket Sampling Mode=::NewEnumerator3
    VelocitySpeedLimit=false
   )
   &ShapeLocation(
    Cylinder Height=1
    Cylinder Radius=28.3356
    MaskSpawnGroup=true
    Override Seed=false
    Shape Origin=&Add_Vector(A=Linked() {linkedVariable {name=Engine.Emitter.SimulationPosition, type=NiagaraPosition}}, B=<0 0 13>)
    Shape Primitive=::NewEnumerator1
    Spawn Group Mask=12
    Surface Only=true
    Use Endcaps In Surface Only Mode=false
    UseOwnerScale=true
   )
   &ShapeLocation(
    Box Size=<55 6 1>
    MaskSpawnGroup=true
    Override Seed=false
    Rotation Angle=0.125
    Rotation Axis=<0 0 1>
    Rotation Mode=::NewEnumerator1
    Shape Origin=&Add_Vector(A=Linked() {linkedVariable {name=Engine.Emitter.SimulationPosition, type=NiagaraPosition}}, B=<0 0 13>)
    Shape Primitive=::NewEnumerator2
    Spawn Group Mask=13
    Surface Only=false
    UseOwnerScale=true
   ) {module=ShapeLocation001}
   &ShapeLocation(
    Box Size=<55 6 1>
    MaskSpawnGroup=true
    Override Seed=false
    Rotation Angle=-0.125
    Rotation Axis=<0 0 1>
    Rotation Mode=::NewEnumerator1
    Shape Origin=&Add_Vector(A=Linked() {linkedVariable {name=Engine.Emitter.SimulationPosition, type=NiagaraPosition}}, B=<0 0 13>)
    Shape Primitive=::NewEnumerator2
    Spawn Group Mask=14
    Surface Only=false
    UseOwnerScale=true
   ) {module=ShapeLocation002}
   &ScaleSpriteSize(Uniform Curve Scale=&FNTimeOfDayScalar(Night Scale=1.25, Morning Scale=1, Evening Scale=1))
   &ScaleColor(
    Scale Alpha=&Multiply_Float(
     A=&FloatFromCurve(
      FloatCurve=DataInterface() {propertyValues {Curve {keys [
          {interpMode=RCIM_Cubic, tangentMode=RCTM_User, time=0, value=0}
          {arriveTangent=-2.4058, interpMode=RCIM_Cubic, leaveTangent=-2.4058, tangentMode=RCTM_User, time=0.26, value=0.8421}
          {arriveTangent=-1.1775, interpMode=RCIM_Cubic, leaveTangent=-1.1775, time=0.4613, value=0.0645}
          {arriveTangent=-0.1197, interpMode=RCIM_Cubic, leaveTangent=-0.1197, time=0.5898, value=0.4537}
          <at 1 0 RCIM_Cubic>
         ]}}}
      Scale Curve=&RandomRangeFloat(Minimum=0.1, Maximum=0.5)
     )
     B=&FloatFromCurve(
      FloatCurve=DataInterface() {propertyValues {Curve {keys [{tangentMode=RCTM_User, time=0, value=0}, {tangentMode=RCTM_User, time=1, value=1}]}}}
      Scale Curve=&FloatFromCurve(
       FloatCurve=DataInterface() {propertyValues {Curve {keys [
           {arriveTangentWeight=0.1667, interpMode=RCIM_Cubic, leaveTangentWeight=0.1667, tangentMode=RCTM_User, time=1.4, value=0}
           {tangentMode=RCTM_User, time=2, value=1}
          ]}}}
       CurveIndex=$Emitter.Age
      )
     )
    )
    Scale RGB=&VectorFromFloat(Value=&Multiply_Float(
      A=&RandomRangeFloat(Minimum=0.7, Maximum=0.9, Evaluation Type=::NewEnumerator1)
      B=&SelectFloatFromArray(Direct Array Index=$Engine.QualityLevel, Float Selection Array=DataInterface() {propertyValues {FloatData [10,5,1,1,1]}})
     ))
    ScaleA=true
    ScaleRGB=true
   ) {module=ScaleColor001}
   &ScaleColor(
    Scale Alpha=&FNTimeOfDayScalar(Night Scale=1.25, Morning Scale=1.25, Day Scale=1.25, Evening Scale=1.25)
    Scale RGB=&MakeVectorFromLinearColorRGB(Color=$System.TOD)
    ScaleA=true
    ScaleRGB=true
   )
   /SetVariables(
    Particles.CenterFlareColor:LinearColor=&MultiplyLinearColor(
     Linear Color=$Particles.Color
     Scale Factor=&MakeVector4_FromLinearColor(LinearColor=&LerpLinearColors(StartColor=<rgba 8.0099 9.8007 10 1>, EndColor=<rgba 3.4218 10 3.4218 1>, LerpFactor=0))
    )
    Particles.CenterFlareSize:Vector2f=&Multiply_Vector2(A=$Particles.SpriteSize, B=&Vector2DFromFloat(Value=0.65))
    Particles.CoronaFlareColor:LinearColor=&MultiplyLinearColor(
     Linear Color=$Particles.Color
     Scale Factor=&MakeVector4_FromLinearColor(LinearColor=&LerpLinearColors(StartColor=<rgba 0.15 0.15 0.15 1>, EndColor=<rgba 0.0729 0.0729 0.0729 1>, LerpFactor=0))
    )
    Particles.CoronaFlareSize:Vector2f=&Multiply_Vector2(A=$Particles.Initial.SpriteSize, B=&Vector2DFromFloat(Value=0.85))
    Particles.FlickerFlareColor:LinearColor=&MultiplyLinearColor(
     Linear Color=$Particles.Color
     Scale Factor=&MakeVector4_FromLinearColor(LinearColor=&LerpLinearColors(StartColor=<rgba 0.9441 0.8079 1 1>, EndColor=<rgba 0.0337 0.0417 0.0337 50>, LerpFactor=0))
    )
   ) {isSetParameters=true}
  ]
  rendererClasses [&NiagaraSpriteRendererProperties]
  renderers [
   {rendererClass=&NiagaraSpriteRendererProperties}
   {rendererClass=&NiagaraSpriteRendererProperties, rendererIndex=1}
   {rendererClass=&NiagaraSpriteRendererProperties, rendererIndex=2}
  ]
 }}
summary {
 systemName=NS_Backpack_ChewyStaveGerm_Idle
 emitters [{emitterName=Sparkles, rendererClasses [&NiagaraSpriteRendererProperties]}]
 userVariables [
  User.bIsFrontEnd {type=NiagaraBool}
  User.bIsFrontEndPreview {type=NiagaraBool}
 ]
}
